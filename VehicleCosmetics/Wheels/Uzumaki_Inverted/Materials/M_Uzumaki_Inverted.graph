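# Unlit material for the Uzumaki Inverted vehicle wheel cosmetic, with an animated emissive effect.
# Panned noise and gradients combine with vertex-color channels, tire texture, view-angle and screen-height terms, tinted by Color or shifted toward a dark variant by Black.
# The LOD4_5Effect static switch blends in a rim-masked, sine-animated circle built from UV1 and a curve texture.
# The result passes through vehicle damage and de-rez transition functions.
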Material /VehicleCosmetics/Wheels/Uzumaki_Inverted/Materials/M_Uzumaki_Inverted
graph {
 properties {
  bAutomaticallySetUsageInEditor=false
  bUseMaterialAttributes=true
  bUsedWithMeshDeformer=true
  bUsedWithSkeletalMesh=true
  shadingModel=MSM_Unlit
  usedShadingModels=Unlit
 }
 nodes [
  Constant() {r=-0.5}
  n1 = VectorParameter() {parameterName=Color, defaultValue=<rgba 0 0.3567 1 1>}
  n2 = TextureSample(UVs=Panner(Coordinate=TextureCoordinate() {vTiling=3}) {speedX=0.5, speedY=-0.5}) {texture=/VehicleCosmetics/Wheels/_SharedTextures/FX/Universal_Noise_34_RGB}
  n3 = NamedRerouteUsage()
  n4 = Multiply(A=@n2.G, B=@n3)
  n5 = Multiply(A=ConstantBiasScale(None=OneMinus(None=@n2.B)) {bias=-0.9, scale=3}, B=@n3)
  n6 = Subtract(A=@n4, B=@n5)
  n7 = Add(A=Max(A=ConstantBiasScale(None=@n6) {bias=-0.3, scale=2}, B=@n5), B=Clamp(None=ConstantBiasScale(None=@n4) {bias=-0.1, scale=10}))
  n8 = NamedRerouteUsage()
  n9 = TextureSample(UVs=Panner(Coordinate=TextureCoordinate() {uTiling=16}) {speedX=-1}) {texture=/VehicleCosmetics/Wheels/_SharedTextures/FX/LinearGradient_Balanced_Vertical}
  n10 = AppendVector(A=Constant() {r=0.75}, B=Constant())
  n11 = TextureCoordinate()
  n12 = TextureSample(UVs=Panner(Coordinate=Add(A=@n10, B=@n11)) {speedX=-0.5}) {texture=&Universal_Noise_34_RGB}
  n13 = Multiply(A=Multiply(A=ConstantBiasScale(None=@n9.G) {bias=-0.2, scale=10}, B=NamedRerouteUsage()), B=@n12.B)
  n14 = CameraVectorWS()
  n15 = OneMinus(None=ComponentMask(None=Transform(None=@n14) {transformSourceType=TRANSFORMSOURCE_World, transformType=TRANSFORM_Tangent}) {b=true})
  n16 = Subtract(A=Clamp(None=ConstantBiasScale(None=@n15) {bias=-0.45, scale=1}), B=NamedRerouteUsage())
  n17 = Clamp(None=ComponentMask(None=ScreenPosition().ViewportUV) {g=true})
  n18 = Multiply(A=Clamp(None=ConstantBiasScale(None=@n17) {bias=-0.5, scale=2}), B=NamedRerouteUsage())
  n19 = Multiply(A=ConstantBiasScale(None=@n18) {bias=-0.4, scale=1}, B=Constant() {r=0.25})
  n20 = Clamp(None=Add(A=Multiply(B=@n18), B=Clamp(None=@n19)))
  n21 = TextureSample() {texture=/VehicleCosmetics/Wheels/Uzumaki_Inverted/Textures/T_Uzumaki_SE_Tire_RGB}
  n22 = Multiply(A=NamedRerouteUsage(), B=Clamp(None=Add(A=@n21.B, B=@n21.B)))
  n23 = Constant() {r=5}
  n24 = NamedRerouteUsage()
  n25 = Clamp(None=@n22) {maxDefault=0.15}
  n26 = Add(A=Multiply(A=Multiply(A=@n22, B=@n23), B=@n24), B=Add(A=@n25, B=Multiply(A=@n24, B=@n25)))
  n27 = Add(A=Multiply(A=Multiply(A=@n8, B=@n13), B=Constant() {r=2}), B=Add(A=Clamp(None=@n16), B=Add(A=@n20, B=@n26)))
  n28 = Multiply(A=@n1.RGB, B=Add(A=@n7, B=@n27))
  n29 = LinearInterpolate(A=Constant3Vector() {constant=<rgba 0.02 0.02 0.02 0>}, B=Constant3Vector(), Alpha=Multiply(A=@n27, B=Constant() {r=20}))
  n30 = Clamp(None=ConstantBiasScale(None=@n7) {bias=-2.5, scale=1}) {maxDefault=1.5}
  n31 = ScalarParameter() {parameterName=Black}
  n32 = Add(A=ConstantBiasScale(None=@n29) {bias=-0.1, scale=10}, B=@n30)
  n33 = Add(A=LinearInterpolate(A=@n28, B=Add(A=@n29, B=@n30), Alpha=@n31), B=LinearInterpolate(A=@n28, B=Clamp(None=@n32) {maxDefault=2}, Alpha=@n31))
  n34 = StaticSwitchParameter(True=LinearInterpolate(A=@n33, B=NamedRerouteUsage(), Alpha=NamedRerouteUsage()), False=@n33) {group=SpecialEffect, parameterName=LOD4_5Effect}
  MaterialFunctionCall(In=MaterialFunctionCall(Material=SetMaterialAttributes(Emissive Color=@n34) {attributeSetTypes [#1]}) {materialFunction=&MF_BaseVehicle_Damage}.Result) {materialFunction=&MF_VehicleDeRez_Transition}
  n35 = Frac(None=TextureCoordinate() {coordinateIndex=1})
  n36 = AppendVector(A=Constant() {r=-0.3}, B=Constant() {r=-0.7})
  Multiply(A=@n35, B=@n36)
  NamedRerouteDeclaration(None=@n12.B) {name=Panner_B, nodeColor=<rgba 0 0.0391 1 1>}
  n37 = VertexColor()
  NamedRerouteDeclaration(None=@n37.R) {name=VertexColorR, nodeColor=<rgba 1 0 0 1>}
  NamedRerouteDeclaration(None=@n37.G) {name=VertexColorG, nodeColor=<rgba 0 1 0 1>}
  NamedRerouteDeclaration(None=@n37.B) {name=VertexColorB, nodeColor=<rgba 0 0.0012 1 1>}
  NamedRerouteDeclaration(None=@n21.R) {name=RGB_R, nodeColor=<rgba 0 1 0.8359 1>}
  NamedRerouteDeclaration(None=@n1.RGB) {name=Color, nodeColor=<rgba 1 0 0.9141 1>}
  n38 = TextureSample() {samplerType=SAMPLERTYPE_LinearColor, texture=/VehicleCosmetics/Wheels/Uzumaki_Inverted/Textures/T_UzumakiInverted_Rim_MASK1_LOD}
  n39 = Add(A=@n35, B=@n36)
  n40 = DotProduct(A=@n39, B=@n39)
  n41 = ScalarParameter() {defaultValue=3, group=SpecialEffect, parameterName=Circle Sharpness}
  n42 = Multiply(A=Subtract(A=OneMinus(None=@n40), B=ScalarParameter() {defaultValue=0.77, group=SpecialEffect, parameterName=Circle Size}), B=@n41)
  n43 = TextureSampleParameter2D(UVs=SquareRoot(None=Clamp(None=@n42))) {parameterName=Curve Generic, samplerType=SAMPLERTYPE_LinearColor, texture=/VehicleCosmetics/Wheels/Reverb/Textures/Curve_Generic_01_Pack}
  n44 = Clamp(None=@n43.RGB)
  n45 = Add(A=Multiply(A=Constant() {r=0.1}, B=Time()), B=@n44)
  n46 = Multiply(A=Multiply(A=Sine(None=@n45) {period=0.1}, B=@n44), B=@n41)
  n47 = Power(Base=@n43.RGB, Exp=@n41)
  n48 = ComponentMask(None=Clamp(None=Multiply(A=@n46, B=@n47))) {b=true}
  NamedRerouteDeclaration(None=LinearInterpolate(
    A=@n38.R
    B=VectorParameter() {group=SpecialEffect, parameterName=Color2, defaultValue=<rgba 0 0.3567 1 1>}.RGB
    Alpha=LinearInterpolate(B=@n48, Alpha=@n38.R)
   )) {name=LOD4_5_Effect, nodeColor=<rgba 0 1 0.7656 1>}
  TextureSample() {samplerType=SAMPLERTYPE_LinearColor, texture=/VehicleCosmetics/Wheels/Uzumaki_Inverted/Textures/T_UzumakiInverted_Rim_MASK2_LOD}
 ]
}
